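# CameraWorldBlend is a library material function meant to give camera-angle falloffs for the three world axes, for blending textures by view direction.
# Nested static switches, described as choosing reflection over camera vector, feed a Transform, though no such vector nodes appear here.
# Each absolute component passes through a soft SphereMask, is inverted and raised to Blend Power, producing the YZ, XZ and XY outputs.
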MaterialFunction /Engine/Functions/Engine_MaterialFunctions03/Texturing/CameraWorldBlend
graph {
 properties {
  bExposeToLibrary=true
  description="Outputs falloff results for the 3 primary world vectors based on the camera angle.  Can be used to blend between textures based on direction the camera is looking."
  libraryCategoriesText [Misc,Texturing]
 }
 nodes [
  n1 = Constant()
  n2 = LinearInterpolate(Alpha=FunctionInput(Preview=Constant()) {bUsePreviewValueAsDefault=true, inputName=Smooth Reflection Percentage, inputType=FunctionInput_Scalar, sortPriority=3})
  n3 = FunctionInput(Preview=StaticBool()) {
   bUsePreviewValueAsDefault=true
   description="If TRUE, uses ReflectionVector instead of CameraVector (Material's Normal input will effect results)"
   inputName=Use Smoothable Normals
   inputType=FunctionInput_StaticBool
   sortPriority=2
  }
  n4 = FunctionInput(Preview=StaticBool()) {
   bUsePreviewValueAsDefault=true
   description="If TRUE, uses ReflectionVector instead of CameraVector (Material's Normal input will effect results)"
   inputName=Use Reflection Vector
   inputType=FunctionInput_StaticBool
   sortPriority=1
  }
  n5 = Transform(None=StaticSwitch(True=StaticSwitch(True=@n2, Value=@n3), Value=@n4))
  n6 = SphereMask(A=@n1, B=Abs(None=ComponentMask(None=@n5) {r=true})) {attenuationRadius=1, hardnessPercent=0}
  n7 = FunctionInput(Preview=Constant() {r=3}) {bUsePreviewValueAsDefault=true, inputName=Blend Power, inputType=FunctionInput_Scalar}
  FunctionOutput(None=Power(Base=OneMinus(None=@n6), Exp=@n7)) {outputName=YZ True, sortPriority=2}
  n8 = SphereMask(A=@n1, B=Abs(None=ComponentMask(None=@n5) {g=true})) {attenuationRadius=1, hardnessPercent=0}
  FunctionOutput(None=Power(Base=OneMinus(None=@n8), Exp=@n7)) {outputName=XZ True, sortPriority=1}
  n9 = SphereMask(A=@n1, B=Abs(None=ComponentMask(None=@n5) {b=true})) {attenuationRadius=1, hardnessPercent=0}
  FunctionOutput(None=Power(Base=OneMinus(None=@n9), Exp=@n7)) {outputName=XY True}
 ]
}
