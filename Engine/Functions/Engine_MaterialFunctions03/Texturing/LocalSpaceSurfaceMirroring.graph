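# Material function that generates per-axis masks from a surface's local-space orientation, intended for mirroring textures by object topology.
# A static switch picks normal-map input or a flat up vector, transforms it to local space, adds a scalar bias and rounds up with Ceil.
# Local X, Y and Z outputs split a OneMinus node by channel, though the graph shows no link from Ceil into it.
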MaterialFunction /Engine/Functions/Engine_MaterialFunctions03/Texturing/LocalSpaceSurfaceMirroring
graph {
 properties {
  description="Generates mask results based on the local orientation of surfaces.  Useful for mirroring texture information based on the object's topology."
  libraryCategoriesText [Misc,Texturing]
 }
 nodes [
  n1 = TextureObject() {samplerType=SAMPLERTYPE_Normal, texture=/Engine/EngineMaterials/DefaultNormal}
  n2 = FunctionInput(Preview=StaticBool()) {
   bUsePreviewValueAsDefault=true
   description=Use texture-based normals instead of surface normals.
   inputName=Use NormalMap
   inputType=FunctionInput_StaticBool
  }
  n3 = StaticSwitch(
   True=FunctionInput(Preview=TextureSample(Tex=@n1).RGB) {bUsePreviewValueAsDefault=true, description=Normal input for surface calculation, inputName=Normal Map}
   False=Constant3Vector() {constant=<rgba 0 0 1 0>}
   Value=@n2
  )
  n4 = Constant()
  Ceil(None=Add(
    A=Transform(None=@n3) {transformType=TRANSFORM_Local}
    B=FunctionInput(Preview=@n4) {bUsePreviewValueAsDefault=true, description=Biases the mask results., inputType=FunctionInput_Scalar}
   ))
  n5 = OneMinus()
  FunctionOutput(None=ComponentMask(None=@n5) {r=true}) {outputName=Local X}
  FunctionOutput(None=ComponentMask(None=@n5) {g=true}) {outputName=Local Y, sortPriority=1}
  FunctionOutput(None=ComponentMask(None=@n5) {b=true}) {outputName=Local Z, sortPriority=2}
 ]
}
