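# TextureVariation_RotateNormals is a material function that rotates a normal's XY components by a 2D direction vector, apparently to correct normals for rotated texture variation.
# It dots Normal.XY with Vector for X and with the perpendicular (-G, R) for Y, forming a 2x2 rotation without trigonometry.
# Z passes through unchanged, and Vector defaults to (1, 0), which leaves the normal unrotated.
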MaterialFunction /Engine/Functions/Engine_MaterialFunctions03/Texturing/TextureVariation_RotateNormals
graph {nodes [
  n1 = FunctionInput() {inputName=Normal, sortPriority=1}
  n2 = ComponentMask(None=@n1) {g=true, r=true}
  n3 = FunctionInput() {bUsePreviewValueAsDefault=true, inputName=Vector, inputType=FunctionInput_Vector2, previewValue=<1 0 0 1>}
  n4 = MaterialFunctionCall(Float2=@n3) {materialFunction=&BreakOutFloat2Components}
  n5 = AppendVector(A=Multiply(A=@n4.G) {constB=-1}, B=@n4.R)
  FunctionOutput(None=MaterialFunctionCall(X=DotProduct(A=@n3, B=@n2), Y=DotProduct(A=@n5, B=@n2), Z=ComponentMask(None=@n1) {b=true}) {materialFunction=&MakeFloat3}.Result)
 ]}
