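# Material function that converts a camera-relative world position into periodic world space with high precision, captioned as superseded by TransformPosition.
# A Custom node turns the Scale input into a power-of-two tile size using ldexp of its floored value.
# TransformPosition maps translated world to periodic world with that tile size, keeping the origin tied to the camera's tile.
# The result is divided by the tile size and output as XYZ, XY and Z.
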MaterialFunction /Engine/Functions/Engine_MaterialFunctions03/Texturing/PeriodicWorldPosition
graph {
 properties {
  description="Convert a camera-relative position into periodic-world space position with optimal precision. The world is divided into tiles of size pow(2, Scale). The origin is defined by the tile the camera is in. Output positions are relative to this origin. Suitable as a position input for functions whose output is identical when the input is shifted by pow(2, Scale)."
  userExposedCaption="PeriodicWorldPosition (Use TransformPosition instead)"
 }
 nodes [
  n1 = WorldPosition() {worldPositionShaderOffset=WPT_CameraRelative}
  MaterialExpressionFunctionInput_1 = FunctionInput() {bUsePreviewValueAsDefault=true, inputName=Scale, previewValue=<1 1 1 1>}
  n2 = Custom!(None=@MaterialExpressionFunctionInput_1) {
   code="ldexp(1.0, floor(Input))"
   showCode=true
   inputs [{
     inputName=Input
     input {expression=&self:MaterialExpressionFunctionInput_1}
    }]
  }
  n3 = Divide(
   A=TransformPosition!(None=FunctionInput(Preview=@n1.XYZ) {inputName=Camera Relative World Position}, Periodic World Tile Size=@n2) {transformSourceType=TRANSFORMPOSSOURCE_TranslatedWorld, transformType=TRANSFORMPOSSOURCE_PeriodicWorld}
   B=@n2
  )
  FunctionOutput(None=ComponentMask(None=@n3) {b=true, g=true, r=true}) {outputName=XYZ}
  FunctionOutput(None=ComponentMask(None=@n3) {g=true, r=true}) {outputName=XY, sortPriority=1}
  FunctionOutput(None=ComponentMask(None=@n3) {b=true}) {outputName=Z, sortPriority=2}
 ]
}
