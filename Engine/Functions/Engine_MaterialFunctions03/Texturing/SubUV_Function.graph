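# SubUV_Function is a library material function that plays flipbook frames from a single grid texture with smooth blending.
# The fractional Frame input is split into a whole index and blend alpha, with the next frame at index plus one.
# Fmod and floor turn both indices into column and row offsets, scaled by SubImages.
# Two texture samples are lerped into RGB and Alpha outputs, while input UVs pass through unchanged.
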MaterialFunction /Engine/Functions/Engine_MaterialFunctions03/Texturing/SubUV_Function
graph {
 properties {
  bExposeToLibrary=true
  description=Allows SubUV blending of multiple texture frames in a single texture
  libraryCategoriesText [Misc,Texturing]
 }
 nodes [
  n1 = FunctionInput() {inputName=Frame, inputType=FunctionInput_Scalar, sortPriority=-10}
  n2 = Frac(None=@n1) {desc=Frame Lerp}
  n3 = Subtract(A=@n1, B=@n2) {desc=IndexA}
  n4 = FunctionInput() {inputName=SubImages, inputType=FunctionInput_Vector2, sortPriority=-20, previewValue=<2 2 0 0>}
  n5 = ComponentMask(None=@n4) {desc="## Horz Images", r=true}
  n6 = Divide(A=Constant2Vector() {g=1, r=1}, B=@n4) {desc=Scale}
  n7 = ComponentMask(None=@n6) {desc="1 / ## Horz Images", r=true}
  n8 = AppendVector(A=Fmod(A=@n3, B=@n5) {desc=IndexA_Horz}, B=Floor(None=Multiply(A=@n3, B=@n7)) {desc=IndexA_Vert}) {desc=IndexA_Combined}
  n9 = FunctionInput(Preview=TextureCoordinate()) {bUsePreviewValueAsDefault=true, inputName=UVs, inputType=FunctionInput_Vector2, sortPriority=-30, previewValue=<2 2 0 0>}
  n10 = FunctionInput(Preview=TextureObject() {texture=/Engine/EngineMaterials/DefaultDiffuse}) {bUsePreviewValueAsDefault=true, inputName=Texture, inputType=FunctionInput_Texture2D, sortPriority=-40}
  n11 = TextureSample(UVs=Multiply(A=Add(A=@n8, B=@n9), B=@n6) {desc=UV_A}, Tex=@n10) {texture=&DefaultDiffuse}
  n12 = Add(B=@n3) {constA=1, desc=IndexB}
  n13 = AppendVector(A=Fmod(A=@n12, B=@n5) {desc=IndexB_Horz}, B=Floor(None=Multiply(A=@n12, B=@n7)) {desc=IndexB_Vert}) {desc=IndexB_Combined}
  n14 = TextureSample(UVs=Multiply(A=Add(A=@n9, B=@n13), B=@n6) {desc=UV_B}, Tex=@n10) {texture=&DefaultDiffuse}
  FunctionOutput(None=LinearInterpolate(A=@n11.RGB, B=@n14.RGB, Alpha=@n2)) {outputName=RGB}
  FunctionOutput(None=LinearInterpolate(A=@n11.A, B=@n14.A, Alpha=@n2)) {outputName=Alpha}
  FunctionOutput(None=@n9) {outputName=UVs, sortPriority=3}
 ]
}
