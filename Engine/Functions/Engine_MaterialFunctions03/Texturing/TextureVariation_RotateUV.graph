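# Material function for texture variation that rotates UV coordinates by a 2D direction vector and shifts them by an offset.
# It multiplies U by the Vector input and V by its perpendicular (-Y, X), sums both, then adds Offset.
# This gives a rotation without trig nodes, and a non-unit Vector also scales the UVs.
# UVs default to the texture coordinates, and the default Vector leaves them unrotated.
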MaterialFunction /Engine/Functions/Engine_MaterialFunctions03/Texturing/TextureVariation_RotateUV
graph {nodes [
  n1 = MaterialFunctionCall(Float2=FunctionInput(Preview=TextureCoordinate()) {bUsePreviewValueAsDefault=true, inputName=UVs, inputType=FunctionInput_Vector2}) {materialFunction=&BreakOutFloat2Components}
  n2 = FunctionInput() {bUsePreviewValueAsDefault=true, inputName=Vector, inputType=FunctionInput_Vector2, sortPriority=2, previewValue=<1 0 0 1>}
  n3 = Multiply(A=@n2, B=@n1.R)
  n4 = MaterialFunctionCall(Float2=@n2) {materialFunction=&BreakOutFloat2Components}
  n5 = Multiply(A=AppendVector(A=Multiply(A=@n4.G) {constB=-1}, B=@n4.R), B=@n1.G)
  FunctionOutput(None=Add(A=FunctionInput() {bUsePreviewValueAsDefault=true, inputName=Offset, inputType=FunctionInput_Vector2, sortPriority=1}, B=Add(A=@n3, B=@n5)))
 ]}
