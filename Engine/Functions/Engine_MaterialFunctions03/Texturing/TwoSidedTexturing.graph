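# TwoSidedTexturing is a library material function that gives each face of a two-sided material its own texture input.
# It clamps TwoSidedSign into a mask that lerps from the Texture Side B input to the Texture Side A input.
# Both inputs default to a shared placeholder texture sample, and the mask is also output separately.
# It has no effect on materials that are not two-sided.
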MaterialFunction /Engine/Functions/Engine_MaterialFunctions03/Texturing/TwoSidedTexturing
graph {
 properties {
  bExposeToLibrary=true
  description="Gives independent texturing control for both sides of a two-sided material.  Will have no effect on non-two sided materials."
  libraryCategoriesText [Misc,Texturing]
 }
 nodes [
  n1 = TextureSample(Tex=TextureObject())
  n2 = Clamp(None=TwoSidedSign())
  FunctionOutput(None=LinearInterpolate(
    A=FunctionInput(Preview=@n1.RGB) {bUsePreviewValueAsDefault=true, description=Input for second side, inputName=Texture Side B, sortPriority=1}
    B=FunctionInput(Preview=@n1.RGB) {bUsePreviewValueAsDefault=true, description=Input for first side, inputName=Texture Side A}
    Alpha=@n2
   )) {description=Result, outputName=Texture Blend Output}
  FunctionOutput(None=@n2) {description="Outputs only the mask, ignoring texturing", outputName=Mask, sortPriority=1}
 ]
}
